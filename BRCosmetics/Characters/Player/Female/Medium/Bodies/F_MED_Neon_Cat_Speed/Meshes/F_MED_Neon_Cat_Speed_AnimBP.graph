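# Post-process style Anim Blueprint for the F_MED_Neon_Cat_Speed body mesh, copying its attached parent's pose and adding deformation and zipper secondary motion.
# The copied pose runs through two Control Rigs, twists then sculpters.
# Anim Dynamics swings the top and bottom zipper bones under gravity, each followed by a weighted multi-target constraint tying its master bone to nearby torso or shoulder bones.
# The event graph holds no logic.
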AnimBlueprint /BRCosmetics/Characters/Player/Female/Medium/Bodies/F_MED_Neon_Cat_Speed/Meshes/F_MED_Neon_Cat_Speed_AnimBP
target_skeleton=&/F_MED_Neon_Cat_Speed_Skeleton
graphs [
 AnimGraph {nodes [
   n1 = Control Rig(Source=Copy Pose From Mesh() {bUseAttachedParent=true}) {
    alphaBoolBlend {blendOption=Linear}
    controlRigAssetReference {blueprintRigClass=&/F_MED_Neon_Cat_Speed_Deform_CtrlRig_twists.F_MED_Neon_Cat_Speed_Deform_CtrlRig_twists_C}
   }
   n2 = Control Rig(Source=@n1) {
    alphaBoolBlend {blendOption=Linear}
    controlRigAssetReference {blueprintRigClass=&/F_MED_Neon_Cat_Speed_Deform_CtrlRig_sculpters.F_MED_Neon_Cat_Speed_Deform_CtrlRig_sculpters_C}
   }
   n3 = Anim Dynamics - Bone: dyn_zipperTop(ComponentPose=Local To Component(LocalPose=@n2), Alpha=1, ExternalForce="0,0,0") {
    angularDampingOverride=0.99
    bOverrideAngularDamping=true
    bUseGravityOverride=true
    simulationSpace=RootRelative
    alphaBoolBlend {blendOption=Linear}
    boundBone {boneName=dyn_zipperTop}
    componentLinearAccScale=<0.5 0.5 0.5>
    componentLinearVelScale=<0.5 0.5 0.5>
    gravityOverride {z=-980}
    physicsBodyDefinitions [{
      boundBone {boneName=dyn_zipperTop}
      boxExtents=<2 10 10>
      constraintSetup {angularLimitsMax=<2 5 30>, angularLimitsMin=<-2 -30 -30>, angularTarget=<0 0 0>}
      localJointOffset=<1 0 0>
     }]
   }
   n4 = "Constraint - dyn_zipperTopMSTR (PPPPPP)"!(
    ComponentPose=@n3
    Alpha=1
    ConstraintWeights_0=0.1312
    ConstraintWeights_1:float=0.4609
    ConstraintWeights_2:float=0.1442
    ConstraintWeights_3:float=0.2355
    ConstraintWeights_4:float=0.0022
    ConstraintWeights_5:float=0.026
   ) -> (Pose) {
    alphaBoolBlend {blendOption=Linear}
    boneToModify {boneName=dyn_zipperTopMSTR}
    constraintSetup [
     {targetBone {boneName=spine_04}}
     {targetBone {boneName=spine_05}}
     {targetBone {boneName=clavicle_r}}
     {targetBone {boneName=deform_pectoral_fr_r}}
     {targetBone {boneName=upperarm_twist_01_r}}
     {targetBone {boneName=deform_deltoid_fr_r}}
    ]
    constraintWeights [1,1,1,1,1,1]
   }
   n5 = Anim Dynamics - Bone: dyn_zipperBottom(ComponentPose=@n4, Alpha=1, ExternalForce="0,0,0") {
    angularDampingOverride=0.9999
    bOverrideAngularDamping=true
    bUseGravityOverride=true
    simulationSpace=RootRelative
    alphaBoolBlend {blendOption=Linear}
    boundBone {boneName=dyn_zipperBottom}
    componentLinearAccScale=<0.2 0.2 0.2>
    componentLinearVelScale=<0.2 0.2 0.2>
    gravityOverride {z=-980}
    physicsBodyDefinitions [{
      boundBone {boneName=dyn_zipperBottom}
      boxExtents=<2 10 10>
      constraintSetup {angularLimitsMax=<2 5 30>, angularLimitsMin=<-2 -30 -30>, angularTarget=<0 0 0>}
      localJointOffset=<1 0 0>
     }]
   }
   n6 = "Constraint - dyn_zipperBottomMSTR (PPP)"!(ComponentPose=@n5, Alpha=1, ConstraintWeights_0=0.3965, ConstraintWeights_1:float=0.5935, ConstraintWeights_2:float=0.01) -> (Pose) {
    alphaBoolBlend {blendOption=Linear}
    boneToModify {boneName=dyn_zipperBottomMSTR}
    constraintSetup [{targetBone {boneName=spine_04}}, {targetBone {boneName=spine_05}}, {targetBone {boneName=deform_pectoral_fr_r}}]
    constraintWeights [1,1,1]
   }
   Output Pose(Result=Component To Local(ComponentPose=@n6)) {layerGroup=None}
  ]}
 EventGraph {nodes [Event BlueprintUpdateAnimation()]}
]
variables []
